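# Animation Blueprint for the Galileo Ferry male medium player head cosmetic, targeting the shared Fortnite base head export skeleton.
# The name marks it as a child Blueprint, though no parent class appears here.
# Its EventGraph is empty and it declares no variables, so it adds no visible logic of its own.
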AnimBlueprint /BRCosmetics/Characters/Player/Male/Medium/Heads/M_MED_Galileo_Ferry_Head/Meshes/M_MED_Ferry_Head_AnimBP_Child_SG
target_skeleton=/Game/Characters/Player/Common/Fortnite_Base_Head/Fortnite_Base_Head_Export_Skeleton
graphs [EventGraph {links [], nodes []}]
variables []
